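; Knockback ability for the Play As Villain container shockwave, built on a parent knockback ability, that lets vehicles take different knockback than other actors.
; On activation, a vehicle avatar has the parent's XY magnitude, Z magnitude and Z add values overwritten with versions scaled by a vehicle multiplier, before parent activation runs.
; GetAbilityInstigator returns the granting effect context's source object as an Actor, printing an error if the cast fails.
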(event Ability|EventActivateAbility
  ;[1 1 1] Vehicles could have different knockback than other actors
  (if (Utilities|IsA (Ability|GetAvatarActorfromActorInfo) "/Script/FortniteGame.FortAthenaVehicle")
    (CallFunction|UpdateKnockbackValuesforVehicle))
  (|CallParentFunction))

(fn GetAbilityInstigator
  (bind _asactor (Utilities|Casting|CastToActor (Ability|EffectContext|GetSourceObject (Ability|GetGrantedbyEffectContext)))
    (:then
      (return _asactor))
    (:CastFailed
      (Development|PrintString "Error: Knockback failed to get instigator"))))

(fn UpdateKnockbackValuesForVehicle
  (Variables|Default|SetKnockbackMultiplier (ScalableFloat|GetValueatLevel (Variables|Default|GetVehicleKnockbackMultiplier_Hf)))
  (Variables|KnockbackParent|SetRow_XYKnockMag (Utilities|Struct|MakeScalableFloat (* (ScalableFloat|GetValueatLevel (Variables|KnockbackParent|GetRow_XYKnockMag)) (Variables|Default|GetKnockbackMultiplier))))
  (Variables|KnockbackParent|SetRow_ZKnockMag (Utilities|Struct|MakeScalableFloat (* (ScalableFloat|GetValueatLevel (Variables|KnockbackParent|GetRow_ZKnockMag)) (Variables|Default|GetKnockbackMultiplier))))
  (Variables|KnockbackParent|SetRow_ZAdd (Utilities|Struct|MakeScalableFloat (* (ScalableFloat|GetValueatLevel (Variables|KnockbackParent|GetRow_ZAdd)) (Variables|Default|GetKnockbackMultiplier)))))
