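; Play-as-villain tactical shield ability that adjusts overshield behavior around the UpgradedOvershield.Active tag and the No Build BR playlist.
; After parent activation, tag wait tasks feed delta gates that toggle the overshield bar, keeping it visible on removal only in No Build.
; Overrides make max overshield and bar visibility playlist-dependent, with the maximum read from a scalable float row.
; The map icon helper only validates the villain pawn component, with no further action.
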(event Ability|EventActivateAbility
  (bind _returnvalue (CallFunction|GetAvatarasFortPlayerPawn))
  (|CallParentFunction)
  (Ability|Tasks|WaitGameplayTagAddToActor (CallFunction|GetAvatarasFortPlayerPawn) "(TagName=\"PlayAsVillain.UpgradedOvershield.Active\")"
    (:then
      (Ability|Tasks|WaitGameplayTagRemoveFromActor _returnvalue "(TagName=\"PlayAsVillain.UpgradedOvershield.Active\")"
        (:Removed
          (Utilities|FlowControl|DeltaGateWildcard (Variables|Default|GetIsUpgradedOvershieldActive)
            (:Success
              (CallFunction|SetMapIconVisibility false)
              (bind _asfort_player_state_zone (Utilities|Casting|CastToFortPlayerStateZone (Variables|/Script/Engine.Pawn|GetPlayerState _returnvalue)))
              (Class|FortPlayerStateZone|SetOvershieldBarVisible _asfort_player_state_zone (Class|FortGameStateAthena|PlaylistHasTag (Fortnite|Athena|GameState|GetGameStateAthena) (GameplayTags|MakeLiteralGameplayTag "(TagName=\"Athena.Playlist.NoBuildBR\")"))))))))
    (:Added
      (Utilities|FlowControl|DeltaGateWildcard (Variables|Default|GetIsUpgradedOvershieldActive) true
        (:Success
          (CallFunction|SetMapIconVisibility true)
          (if (not (Class|FortGameStateAthena|PlaylistHasTag (Fortnite|Athena|GameState|GetGameStateAthena) (GameplayTags|MakeLiteralGameplayTag "(TagName=\"Athena.Playlist.NoBuildBR\")")))
            (bind _asfort_player_state_zone_2 (Utilities|Casting|CastToFortPlayerStateZone (Variables|/Script/Engine.Pawn|GetPlayerState (CallFunction|GetAvatarasFortPlayerPawn))))
            (Class|FortPlayerStateZone|SetOvershieldBarVisible _asfort_player_state_zone_2 true)))))))

(fn CleanupMaxOvershield
  (params (PawnRef :type Fort%20Pawn%20Object%20Reference))
  (CallFunction|SetMaxOvershield PawnRef))

(fn OverrideMaxOvershield
  (params (PawnRef :type Fort%20Pawn%20Object%20Reference))
  (CallFunction|SetMaxOvershield PawnRef))

(fn OverrideSetOvershieldVisibility
  (params (PlayerStateRef :type Fort%20Player%20State%20Zone%20Object%20Reference) (Visible :type Boolean))
  (|CallParentFunction PlayerStateRef (select (Class|FortGameStateAthena|PlaylistHasTag (Fortnite|Athena|GameState|GetGameStateAthena) (GameplayTags|MakeLiteralGameplayTag "(TagName=\"Athena.Playlist.NoBuildBR\")")) Visible false)))

(fn SetMapIconVisibility
  (params (IsAbilityActive :type Boolean))
  (bind _returnvalue (Ability|GetAvatarActorfromActorInfo))
  (Utilities|IsValid _returnvalue
    (:Is%20Valid
      (Utilities|IsValid (Actor|GetComponentbyClass _returnvalue "/PlayAsVillain/Gameplay/B_PlayAsVillain_FortPawnComponent.B_PlayAsVillain_FortPawnComponent_C")))))

(fn SetMaxOvershield
  (params (PawnRef :type Fort%20Pawn%20Object%20Reference))
  (FortPawn|SetMaxOvershield PawnRef (Math|Float|SelectFloat (ScalableFloat|GetValueatLevelwithContext (Variables|Rows|GetRow_MaxTacticalOvershield)) 0.0 (Class|FortGameStateAthena|PlaylistHasTag (Fortnite|Athena|GameState|GetGameStateAthena) (GameplayTags|MakeLiteralGameplayTag "(TagName=\"Athena.Playlist.NoBuildBR\")")))))
